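; Indigo boost shockwave projectile that explodes right after spawning on the server, knocking pawns away, while remote clients attach a jump trail to the instigator's cosmetic part.
; On explosion it skips the instigator's teammates and knockback-immune pawns, then detaches remote control, cancels fall damage and launches using Indigo curve-table values with a minimum lift.
; A second pass launches AI pawns using KnockGrenade values.
; Explosion effects and near or distant sounds play cosmetically.
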(event EventBeginPlay
  ;[1 1 1] Begin Play
  (Networking|SwitchHasAuthority
    (:Authority
      (Utilities|FlowControl|Delay 0.0)
      (Delivery|DoExplosion))
    (:Remote
      (bind _asfort_player_pawn (Utilities|Casting|CastToFortPlayerPawn (Variables|Actor|GetInstigator)))
      (bind (_pawn _part _meshpart) (Cosmetics|WaitForCosmeticSpecificPartApplied _asfort_player_pawn true "MiscOrTail")
        (:OnPartApplied
          (Effects|SpawnEmitterAttached "/Figment_Martin/Gameplay/Martin_Items/Indigo/FX/P_Ashton_Indigo_JumpAirTrail.P_Ashton_Indigo_JumpAirTrail" _meshpart "root"))))))

(event EventOnExploded
  (params (HitActors :type Array%20of%20Actor%20Object%20References) (HitResults :type Array%20of%20Hit%20Result%20Structures))
  (bind _target (Variables|BPrjGrenade|GetTarget))
  ;[1 1 1] OnExploded - Emitter, Sound, Camera Shake
  (Utilities|FlowControl|Sequence
    (:then_0
      (for _array_element HitActors
        (bind _asfort_pawn (Utilities|Casting|CastToFortPawn _array_element))
        (Variables|BPrjGrenade|SetTarget _asfort_pawn)
        (DataTable|EvaluateCurveTableRow "/Game/Athena/Balance/DataTables/AthenaGameData.AthenaGameData" "Default.KnockGrenade.FriendlyFire"
          (:RowFound
            _)
          (:RowNotFound
            _))
        ;[1 1 1] Perform a team check. Same team?
        (if (and (Utilities|NotEqual(Object) (Variables|Actor|GetInstigator) (Variables|BPrjGrenade|GetTarget)) (Team|OnSameTeam (Variables|BPrjGrenade|GetTarget) (Variables|Actor|GetInstigator)))
          _
          (elif (GameplayTags|HasMatchingGameplayTag (Variables|BPrjGrenade|GetTarget) "(TagName=\"Gameplay.Status.KnockbackImmunity\")")
            _
            (else
              ;[0 0.4531 0.2084] Launch
              (Utilities|FlowControl|Sequence
                (:then_0
                  (CallFunction|HandlePawnDetachRC)
                  (Variables|/Script/FortniteGame.FortPawn|SetbIgnoreNextFallingDamage true (Variables|BPrjGrenade|GetTarget))
                  (Components|Movement|StopMovementImmediately (Variables|/Script/Engine.Character|GetCharacterMovement _target))
                  ;;[1 0 0] NOT SET UP YET
                  (bind _outxy (DataTable|EvaluateCurveTableRow "/Game/Athena/Balance/DataTables/AthenaGameData.AthenaGameData" "Default.Ashton.Indigo.BoostShockwave.ImpulseStrength")
                    (:RowFound
                      (bind _outxy_2 (DataTable|EvaluateCurveTableRow "/Game/Athena/Balance/DataTables/AthenaGameData.AthenaGameData" "Default.Ashton.Indigo.BoostShockwave.AddToZBeforeLaunch")
                        (:RowFound
                          (Character|LaunchCharacter _target (Math|Vector|MakeVector (.x (* (Math|Vector|Normalize (- (Transformation|GetActorLocation _target) (Transformation|GetActorLocation))) _outxy)) (.y (* (Math|Vector|Normalize (- (Transformation|GetActorLocation _target) (Transformation|GetActorLocation))) _outxy)) (Math|Float|Max(Float) (+ (.z (* (Math|Vector|Normalize (- (Transformation|GetActorLocation _target) (Transformation|GetActorLocation))) _outxy)) _outxy_2) 1e+03)) true true)
                          (Networking|SwitchHasAuthority
                            (:Authority
                              (bind _asfort_player_pawn_athena (Utilities|Casting|CastToFortPlayerPawnAthena _target))
                              (Class|FortPlayerPawnAthena|SetFallInstigator _asfort_player_pawn_athena (Game|GetInstigatorController) (Variables|Tags|GetDefaultTags)))))))))
                (:then_1
                  (Effects|SpawnEmitterAttached :EmitterTemplate "/Game/Athena/Items/Consumables/KnockGrenade/P_Athena_KnockGrenade_Player_Trails.P_Athena_KnockGrenade_Player_Trails" :AttachToComponent (Variables|/Script/Engine.Character|GetMesh (Variables|BPrjGrenade|GetTarget)) :AttachPointName "root" :Scale "1.000000,1.000000,1.000000" :LocationType "SnapToTarget" :bAutoDestroy true :PoolingMethod "None" :bAutoActivate true)))))))
      ;;[1 0.8789 0.2904] FortAIPawn
      (for _array_element_2 HitActors
        (bind _asfort_aipawn (Utilities|Casting|CastToFortAIPawn _array_element_2))
        (Variables|Default|SetFortAIPawn _asfort_aipawn)
        (CallFunction|LaunchFortAIPawn)))
    (:then_1
      (Networking|CanExecuteCosmeticEvents
        (:True
          (Effects|SpawnEmitteratLocation (Variables|BPrjGrenade|GetP_Explosion) (Transformation|GetActorLocation) (Transformation|GetActorRotation))
          (if (Pawn|IsLocallyControlled (Game|GetInstigator))
            (Audio|PlaySoundatLocation (Variables|Sounds|GetCue_CloseSound) (Transformation|GetActorLocation) "0, 0, 0" 0.6) ; Play Close Explosion
            (Audio|Components|Audio|FadeOut (Variables|B_Prj_Grenade|GetGrenadeFuse_AudioComponent) 0.2 0.0 "Linear" :!class /Script/Engine.AudioComponent) ; Fade Out Fuse Sound
            (else
              (Audio|PlaySoundatLocation (Variables|Sounds|GetCue_DistantSound) (Transformation|GetActorLocation) "0, 0, 0" 0.4) ; Play Distant Explosion
            )))))))

(event Custom|LaunchFortAIPawn
  (bind _fortaipawn (Variables|Default|GetFortAIPawn))
  ;[0.967 1 0.5015] Launch FortAIPawn
  (Utilities|FlowControl|Sequence
    (:then_0
      (bind _outxy (DataTable|EvaluateCurveTableRow "/Game/Athena/Balance/DataTables/AthenaGameData.AthenaGameData" "Default.KnockGrenade.LaunchVelocity")
        (:RowFound
          (bind _outxy_2 (DataTable|EvaluateCurveTableRow "/Game/Athena/Balance/DataTables/AthenaGameData.AthenaGameData" "Default.KnockGrenade.AddToZBeforeLaunch")
            (:RowFound
              (Utilities|IsValid _fortaipawn
                (:Is%20Valid
                  (Components|Movement|StopMovementImmediately (Variables|/Script/Engine.Character|GetCharacterMovement _fortaipawn))
                  (Character|LaunchCharacter _fortaipawn (Math|Vector|MakeVector (.x (* (Math|Vector|Normalize (- (Transformation|GetActorLocation (Variables|Default|GetFortAIPawn)) (Variables|BPrjGrenade|GetHitLocation))) _outxy)) (.y (* (Math|Vector|Normalize (- (Transformation|GetActorLocation (Variables|Default|GetFortAIPawn)) (Variables|BPrjGrenade|GetHitLocation))) _outxy)) (Math|Float|Max(Float) (+ (.z (* (Math|Vector|Normalize (- (Transformation|GetActorLocation (Variables|Default|GetFortAIPawn)) (Variables|BPrjGrenade|GetHitLocation))) _outxy)) _outxy_2) 5e+02)) true true))))))))
    (:then_1
      (Effects|SpawnEmitterAttached :EmitterTemplate "/Game/Athena/Items/Consumables/KnockGrenade/P_Athena_KnockGrenade_Player_Trails.P_Athena_KnockGrenade_Player_Trails" :AttachToComponent (Variables|/Script/Engine.Character|GetMesh (Variables|Default|GetFortAIPawn)) :AttachPointName "root" :Scale "1.000000,1.000000,1.000000" :LocationType "SnapToTarget" :bAutoDestroy true :PoolingMethod "None" :bAutoActivate true))))

(fn HandlePawnDetachRC
  (bind _asfort_player_pawn (Utilities|Casting|CastToFortPlayerPawn (Variables|BPrjGrenade|GetTarget)))
  (RemoteControlPawn|ForceDetachfromRemoteControlledPawn _asfort_player_pawn))

(fn ConstructionScript
  (Networking|CanExecuteCosmeticEvents
    (:True
      (Components|Movement|SetUpdatedComponent (Variables|Default|GetRotatingMovement) 0))))
