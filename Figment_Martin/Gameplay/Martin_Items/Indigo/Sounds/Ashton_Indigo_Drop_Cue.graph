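# Sound cue for the drop sound of the Ashton Indigo item.
# A single wave player feeds the output directly with the Ashton Milo Launcher drop wave, at reduced volume and with no randomization.
# It uses the shared weapon-drop attenuation asset and is routed through the third-person ranged weapon fire sound class.
# A concurrency override blocks new instances once the limit is reached instead of stopping playing ones.
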SoundCue /Figment_Martin/Gameplay/Martin_Items/Indigo/Sounds/Ashton_Indigo_Drop_Cue
root=WavePlayer_1
nodes [{id=WavePlayer_1, properties {soundWaveAssetPtr=&/Ashton_Milo_Launcher_Drop}}]
properties {
 attenuationSettings=&Fort_Weapon_Drop_Att
 bOverrideConcurrency=true
 maxDistance=3000
 soundClassObject=&Fort_WeaponRanged_Fire_3P_PSM
 volumeMultiplier=0.4
 attenuationOverrides {distanceAlgorithm=NaturalSound, falloffDistance=4000, attenuationShapeExtents {x=1000}}
 concurrencyOverrides {resolutionRule=PreventNew}
}
