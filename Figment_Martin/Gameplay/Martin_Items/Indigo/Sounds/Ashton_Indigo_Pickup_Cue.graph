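# Sound cue that plays the pickup sound for the Indigo item (Ashton naming) in Figment_Martin.
# A single wave player feeds the root directly, with no randomization or mixing, at reduced volume.
# It borrows the shared weapon reload attenuation and the third-person ranged weapon fire sound class.
# A concurrency override refuses new instances once the limit is reached rather than cutting off playing ones.
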SoundCue /Figment_Martin/Gameplay/Martin_Items/Indigo/Sounds/Ashton_Indigo_Pickup_Cue
root=WavePlayer_2
nodes [{id=WavePlayer_2, properties {soundWaveAssetPtr=&/Ashton_Indigo_Pickup_01}}]
properties {
 attenuationSettings=&Fort_Weapon_Reload_Att
 bOverrideConcurrency=true
 maxDistance=2250
 soundClassObject=&Fort_WeaponRanged_Fire_3P_PSM
 volumeMultiplier=0.7
 attenuationOverrides {distanceAlgorithm=NaturalSound, falloffDistance=4000, attenuationShapeExtents {x=1000}}
 concurrencyOverrides {resolutionRule=PreventNew}
}
